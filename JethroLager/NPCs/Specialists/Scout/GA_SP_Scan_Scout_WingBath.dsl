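; GA_SP_Scan_Scout_WingBath is a gameplay ability for the Scout specialist's WingBath scan, which makes the player pawn show a speech bubble.
; On activation it runs the parent ability, casts player pawn 0 to BP_PlayerPawn_SP_Scout_WingBath, and plays the OnScan WingBath conversation entry through the shared AI speech bubble function.
; The end-ability event holds only a disabled parent call.
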(event Ability|EventActivateAbility
  (|CallParentFunction)
  (bind _asbp_player_pawn_sp_scout_wing_bath (Utilities|Casting|CastToBP_PlayerPawn_SP_Scout_WingBath (Game|GetPlayerPawn 0)))
  (Class|BPFLAICommonFunctions|PlaySpeechBubble _asbp_player_pawn_sp_scout_wing_bath "(TagName=\"Jethro.Conversation.Entry.WingBath.OnScan\")"))

(event Ability|EventOnEndAbility
  (params (bWasCancelled :type Boolean))
  (|CallParentFunction bWasCancelled) ; This node is disabled and will not be called.\nDrag off pins to build functionality.
)
